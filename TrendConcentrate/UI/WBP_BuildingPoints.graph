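# Widget that shows build points as a row of up to three TrendConcentrate build-icon pips, all collapsed by default.
# UpdatePips stores the passed amount, then checks each pip image in turn, stopping with an on-screen error print if one is invalid.
# Chained exact float comparisons set pips Visible or Hidden for totals of zero to three; any other value leaves them unchanged.
# The event graph holds only empty stub events, so updates rely on UpdatePips being called.
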WidgetBlueprint /TrendConcentrate/UI/WBP_BuildingPoints
graphs [
 EventGraph {nodes [
   Event PreConstruct()
   Event Construct()
   Event Tick()
  ]}
 UpdatePips {nodes [
   n1 = UpdatePips!() -> (then, BuildPointsAmount:double=~) {is=title K2Node_FunctionEntry}
   > Set Build Points(Build Points:double=@n1) -> (Output_Get:double)
   > n2 = Is Valid(InputObject=Get Pip1() -> (Pip1:Image Object Reference))
   PrintString!(execute=@n2.Is Not Valid, InString=PIP1 INVALID, bPrintToScreen, bPrintToLog, TextColor=(R=1,G=0.1983,B=0,A=1), Duration=10, Key) -> (then)
   n3 = Is Valid(exec=@n2.Is Valid, InputObject=Get Pip2() -> (Pip2:Image Object Reference))
   PrintString(execute=@n3.Is Not Valid, InString=PIP2 INVALID, TextColor=(R=1,G=0.1983,B=0,A=1), Duration=10)
   n4 = Is Valid(exec=@n3.Is Valid, InputObject=Get Pip3() -> (Pip3:Image Object Reference))
   PrintString(execute=@n4.Is Not Valid, InString=PIP3 INVALID, TextColor=(R=1,G=0.1983,B=0,A=1), Duration=10)
   n5 = Branch(execute=@n4.Is Valid, Condition="Equal (Float)"(A=Get Build Points() -> (Build Points:double)))
   > SetVisibility(self=Get Pip1(), InVisibility=Hidden)
   > SetVisibility(self=Get Pip2(), InVisibility=Hidden)
   > SetVisibility(self=Get Pip3(), InVisibility=Hidden)
   n6 = Branch(execute=@n5.else, Condition="Equal (Float)"(A=Get Build Points(), B=1))
   > SetVisibility(self=Get Pip1(), InVisibility=Visible)
   > SetVisibility(self=Get Pip2(), InVisibility=Hidden)
   > SetVisibility(self=Get Pip3(), InVisibility=Hidden)
   n7 = Branch(execute=@n6.else, Condition="Equal (Float)"(A=Get Build Points(), B=2))
   > SetVisibility(self=Get Pip1(), InVisibility=Visible)
   > SetVisibility(self=Get Pip2(), InVisibility=Visible)
   > SetVisibility(self=Get Pip3(), InVisibility=Hidden)
   Branch(execute=@n7.else, Condition="Equal (Float)"(A=Get Build Points(), B=3))
   > SetVisibility(self=Get Pip1(), InVisibility=Visible)
   > SetVisibility(self=Get Pip2(), InVisibility=Visible)
   > SetVisibility(self=Get Pip3(), InVisibility=Visible)
  ]}
]
variables [Build Points {type=number}]
widgets {
 info {inheritedWidgetCount=0, parentClass=&UserWidget, rootWidgetClass=&HorizontalBox, widgetCount=4}
 widgets [
  HorizontalBox_49 {widgetClassPath=&HorizontalBox, widget=HorizontalBox_49}
  Pip1 {
   widgetClassPath=&Image
   bIsVariable=true
   parent=HorizontalBox_49
   widget=Pip1
   properties {
    visibility=Collapsed
    brush {
     bIsDynamicallyLoaded=false
     drawAs=Image
     imageType=FullColor
     mirroring=NoMirror
     resourceName=None
     resourceObject=&/T_UI_Icon_Build_TrendConcentrate
     tiling=NoTile
     imageSize=<64 64>
     margin=<ltrb 0 0 0 0>
     outlineSettings {
      bUseBrushTransparency=false
      roundingType=HalfHeightRadius
      width=0
      color {colorUseRule=UseColor_Specified, specifiedColor=<rgba 0 0 0 0>}
      cornerRadii=<0 0 0 0>
     }
     tintColor {colorUseRule=UseColor_Specified, specifiedColor=<rgba 1 1 1 1>}
     uVRegion {bIsValid=false, max=<0 0>, min=<0 0>}
    }
    renderTransform {angle=0, scale=<0.1 0.1>, shear=<0 0>, translation=<0 0>}
   }
   slot {class=HorizontalBoxSlot, properties {size {sizeRule=Fill, value=1}}}
  }
  Pip2 {
   widgetClassPath=&Image
   bIsVariable=true
   parent=HorizontalBox_49
   widget=Pip2
   properties {
    visibility=Collapsed
    brush {
     bIsDynamicallyLoaded=false
     drawAs=Image
     imageType=FullColor
     mirroring=NoMirror
     resourceName=None
     resourceObject=&T_UI_Icon_Build_TrendConcentrate
     tiling=NoTile
     imageSize=<64 64>
     margin=<ltrb 0 0 0 0>
     outlineSettings {
      bUseBrushTransparency=false
      roundingType=HalfHeightRadius
      width=0
      color {colorUseRule=UseColor_Specified, specifiedColor=<rgba 0 0 0 0>}
      cornerRadii=<0 0 0 0>
     }
     tintColor {colorUseRule=UseColor_Specified, specifiedColor=<rgba 1 1 1 1>}
     uVRegion {bIsValid=false, max=<0 0>, min=<0 0>}
    }
    renderTransform {angle=0, scale=<0.1 0.1>, shear=<0 0>, translation=<0 0>}
   }
   slot {class=HorizontalBoxSlot, properties {size {sizeRule=Fill, value=1}}}
  }
  Pip3 {
   widgetClassPath=&Image
   bIsVariable=true
   parent=HorizontalBox_49
   widget=Pip3
   properties {
    visibility=Collapsed
    brush {
     bIsDynamicallyLoaded=false
     drawAs=Image
     imageType=FullColor
     mirroring=NoMirror
     resourceName=None
     resourceObject=&T_UI_Icon_Build_TrendConcentrate
     tiling=NoTile
     imageSize=<64 64>
     margin=<ltrb 0 0 0 0>
     outlineSettings {
      bUseBrushTransparency=false
      roundingType=HalfHeightRadius
      width=0
      color {colorUseRule=UseColor_Specified, specifiedColor=<rgba 0 0 0 0>}
      cornerRadii=<0 0 0 0>
     }
     tintColor {colorUseRule=UseColor_Specified, specifiedColor=<rgba 1 1 1 1>}
     uVRegion {bIsValid=false, max=<0 0>, min=<0 0>}
    }
    renderTransform {angle=0, scale=<0.1 0.1>, shear=<0 0>, translation=<0 0>}
   }
   slot {class=HorizontalBoxSlot, properties {size {sizeRule=Fill, value=1}}}
  }
 ]
}
